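# Sound cue for the intro of the small crowd loop layer in the Impulse emote's foley audio.
# A Random node with a single equally weighted child plays one wave player, so there is effectively no variation.
# It uses emote attenuation, foley sound class and property sheet, plays when silent, and limits concurrency to one instance.
# It also sends to the music audio-analysis submix.
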SoundCue /BRCosmetics/Sounds/Emotes/Impulse/EmoteFoley_Impulse_Crowd_Loop_Intro_Cue
root=Random_0
nodes [
 {id=Random_0, children [WavePlayer_0], properties {weights [1]}}
 {id=WavePlayer_0, properties {soundWaveAssetPtr=&/Emote_Impulse_Crowd_Loop_Layer_Small_Intro}}
]
properties {
 attenuationSettings=&Fort_Emotes_Athena_Att
 audioPropertiesSheet=/BRCosmetics/Sounds/Settings/PropertySheets/Emotes/APS_Cosmetic_Emote_Foley
 maxDistance=1600
 soundClassObject=/Game/Sounds/FortSoundClassesAndModes/Fort_Foley_Emote
 virtualizationMode=PlayWhenSilent
 volumeMultiplier=1
 concurrencyOverrides {bVolumeScaleCanRelease=true, maxCount=1, retriggerTime=1}
 soundSubmixSends [{minSendLevel=0, soundSubmix=&Fortnite_AudioAnalysis_Music}]
}
