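# Sound cue for the scarf poof foley in the Impulse emote, routed through emote foley sound class and property sheet with emote attenuation.
# A Random node picks among five equally weighted poof waves and feeds a root Modulator that pins volume, leaving pitch at its default range.
# The cue keeps playing when silent through virtualization and has a submix send to the music audio-analysis submix.
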SoundCue /BRCosmetics/Sounds/Emotes/Impulse/EmoteFoley_Impulse_Scarf_Poof_Cue
root=Modulator_0
nodes [
 {id=Modulator_0, children [Random_0], properties {volumeMax=1, volumeMin=1}}
 {id=Random_0, children [WavePlayer_0,WavePlayer_1,WavePlayer_2,WavePlayer_3,WavePlayer_4], properties {weights [1,1,1,1,1]}}
 {id=WavePlayer_0, properties {soundWaveAssetPtr=&/Emote_Impulse_Scarf_Poof_05}}
 {id=WavePlayer_1, properties {soundWaveAssetPtr=&/Emote_Impulse_Scarf_Poof_01}}
 {id=WavePlayer_2, properties {soundWaveAssetPtr=/Game/Athena/Sounds/Emotes/Impulse/Emote_Impulse_Scarf_Poof_02}}
 {id=WavePlayer_3, properties {soundWaveAssetPtr=&/Emote_Impulse_Scarf_Poof_03}}
 {id=WavePlayer_4, properties {soundWaveAssetPtr=&/Emote_Impulse_Scarf_Poof_04}}
]
properties {
 attenuationSettings=&Fort_Emotes_Athena_Att
 audioPropertiesSheet=/BRCosmetics/Sounds/Settings/PropertySheets/Emotes/APS_Cosmetic_Emote_Foley
 maxDistance=1600
 soundClassObject=/Game/Sounds/FortSoundClassesAndModes/Fort_Foley_Emote
 virtualizationMode=PlayWhenSilent
 volumeMultiplier=0.8
 concurrencyOverrides {maxCount=1, resolutionRule=PreventNew, retriggerTime=0.5, voiceStealReleaseTime=3}
 soundSubmixSends [{minSendLevel=0, soundSubmix=&Fortnite_AudioAnalysis_Music}]
}
