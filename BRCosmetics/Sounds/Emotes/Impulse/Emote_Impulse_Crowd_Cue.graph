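# Looping crowd sound cue for the Impulse emote, layering crowd intensity by follower count.
# A parameter crossfade on FollowerCount feeds from three looping wave players: Small, Medium and Large.
# Small plays from zero, Medium fades in up to one follower and Large up to two, stacking rather than replacing each other.
# It uses shared emote attenuation, an emote music sound class and an emote foley concurrency set.
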SoundCue /BRCosmetics/Sounds/Emotes/Impulse/Emote_Impulse_Crowd_Cue
root=ParamCrossFade_1
nodes [
 {
  id=ParamCrossFade_1
  children [WavePlayer_6,WavePlayer_8,WavePlayer_7]
  properties {
   paramName=FollowerCount
   crossFadeInput [
    {fadeInDistanceEnd=0, fadeInDistanceStart=0, fadeOutDistanceEnd=100, fadeOutDistanceStart=100, volume=1}
    {fadeInDistanceEnd=1, fadeInDistanceStart=0, fadeOutDistanceEnd=100, fadeOutDistanceStart=100, volume=1}
    {fadeInDistanceEnd=2, fadeInDistanceStart=1, fadeOutDistanceEnd=100, fadeOutDistanceStart=100, volume=1}
   ]
  }
 }
 {id=WavePlayer_6, properties {bLooping=true, soundWaveAssetPtr=&/Emote_Impulse_Crowd_Loop_Layer_Small}}
 {id=WavePlayer_7, properties {bLooping=true, soundWaveAssetPtr=&/Emote_Impulse_Crowd_Loop_Layer_Large}}
 {id=WavePlayer_8, properties {bLooping=true, soundWaveAssetPtr=&/Emote_Impulse_Crowd_Loop_Layer_Medium}}
]
properties {
 attenuationSettings=&Fort_Emotes_Athena_Att
 maxDistance=1600
 soundClassObject=/Game/Sounds/FortSoundClassesAndModes/Fort_Music_Emote_1P_PSM
 concurrencyOverrides {maxCount=1, resolutionRule=StopOldest, retriggerTime=0.1}
 concurrencySet [/Game/Sounds/Concurrency_Rules/Athena_EmoteFoley_Conc]
}
