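# Front-end sound cue for the Impulse emote, playing a looping large crowd layer.
# An Enveloper at the root ramps the looping wave player from silence to full volume as a fade-in, with a single-key pitch curve.
# The cue uses emote attenuation, a first-person emote music sound class and an emote foley concurrency set, alongside a single-instance, stop-oldest concurrency override.
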SoundCue /BRCosmetics/Sounds/Emotes/Impulse/Emote_Impulse_Crowd_Cue_FrontEnd
root=Enveloper_0
nodes [
 {
  id=Enveloper_0
  children [WavePlayer_7]
  properties {pitchCurve {editorCurveData {keys [{time=0}]}}, volumeCurve {editorCurveData {keys [<at 0 0>,<at 1 1>]}}}
 }
 {id=WavePlayer_7, properties {bLooping=true, soundWaveAssetPtr=&/Emote_Impulse_Crowd_Loop_Layer_Large}}
]
properties {
 attenuationSettings=&Fort_Emotes_Athena_Att
 maxDistance=1600
 soundClassObject=/Game/Sounds/FortSoundClassesAndModes/Fort_Music_Emote_1P_PSM
 concurrencyOverrides {maxCount=1, resolutionRule=StopOldest, retriggerTime=0.1}
 concurrencySet [/Game/Sounds/Concurrency_Rules/Athena_EmoteFoley_Conc]
}
